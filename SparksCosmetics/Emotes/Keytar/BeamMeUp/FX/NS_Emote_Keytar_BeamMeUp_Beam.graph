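# Niagara system for the Keytar BeamMeUp emote: a cyan beam of cone meshes with a light, deploy cone, ground rings, sparks, lightning and shards.
# Skeletal mesh samples from a system mesh and the owner skeleton place the effects, set beam length and attract lightning.
# System-age curves flicker the cones near the end.
# Ground rings and decals follow traced slopes, with fixed position and orientation in front-end preview.
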NiagaraSystem /SparksCosmetics/Emotes/Keytar/BeamMeUp/FX/NS_Emote_Keytar_BeamMeUp_Beam
dependencies {
 usedDataInterfaces [
  &NiagaraDataInterfaceCurve
  &NiagaraDataInterfaceColorCurve
  &NiagaraDataInterfaceVectorCurve
  &NiagaraDataInterfaceCollisionQuery
  &NiagaraDataInterfaceDebugDraw
  &NiagaraDataInterfaceVector2DCurve
 ]
 usedDynamicInputs []
 usedModules [
  &SystemState
  &self:SystemSpawnScript.SystemScriptSource.SystemScriptGraph.NiagaraNodeAssignment_0.SetVariables_#1
 ]
}
emitters {
 ConeLights {
  emitterUpdateScript [
   &EmitterState(
    Inactive Response=::NewEnumerator1
    Life Cycle Mode=::NewEnumerator1
    Loop Behavior=::NewEnumerator1
    Loop Delay=1.7
    Loop Duration=10
    UseLoopDelay=false
   )
   &SpawnBurst_Instantaneous(Use Spawn Probability=false, UseLoopCountLimit=false)
  ]
  particleSpawnScript [
   &InitializeParticle(
    Color=<rgba 0.0075 1 0.9331 0.05>
    Lifetime=3.54
    Mesh Scale=<0.6 0.6 0.6>
    Mesh Scale Mode=::NewEnumerator4
    Sprite Size Mode=::NewEnumerator3
    Uniform Sprite Size=1000
    UsePositionOffset=false
    Write Material Random=true
   )
   &InitialMeshOrientation(Mesh Orientation Mode=::NewEnumerator2, Use Rotation Vector=true)
  ]
  particleUpdateScript [
   &SolveForcesAndVelocity(Clamp Velocity=false, Limit Acceleration=false)
   &ParticleState()
   &SkeletalMeshLocation(
    Bone Sampling Mode=::NewEnumerator4
    MaskSpawnGroup=false
    Orientation Sampling=::NewEnumerator1
    Sampled Position Offset=<0 0 -5>
    Skeletal Mesh=$System.Skeletal Mesh
   ) {module=SkeletalMeshLocation001}
   &SkeletalMeshLocation(
    Bone Sampling Mode=::NewEnumerator4
    MaskSpawnGroup=false
    Orientation Sampling=::NewEnumerator1
    Position Sampling=::NewEnumerator2
    Skeletal Mesh=$User.OwnerSkeleton
   ) {module=SkeletalMeshLocation002}
   &ScaleMeshSize(Scale Factor=&Multiply_Vector(
     A=&MakeVector(
      X=&FloatFromCurve(FloatCurve=DataInterface() {propertyValues {Curve {keys [
           <at 0 0 RCIM_Cubic>
           {arriveTangentWeight=0.0247, interpMode=RCIM_Cubic, leaveTangentWeight=0.0255, tangentMode=RCTM_User, time=0.0174, value=0.05}
           {arriveTangentWeight=0.0589, interpMode=RCIM_Cubic, leaveTangentWeight=0.2882, tangentMode=RCTM_User, time=0.0565, value=0.05}
           {
            arriveTangent=3.9696
            arriveTangentWeight=0.0453
            interpMode=RCIM_Cubic
            leaveTangent=3.9696
            leaveTangentWeight=0.4261
            tangentMode=RCTM_User
            time=0.1319
            value=0.75
           }
           {arriveTangent=0.288, interpMode=RCIM_Cubic, leaveTangent=0.288, time=0.2388, value=0.8}
           {arriveTangentWeight=0.3013, interpMode=RCIM_Cubic, tangentMode=RCTM_User, time=1, value=1}
          ]}}})
      Y=&FloatFromCurve(FloatCurve=DataInterface() {propertyValues {Curve {keys [
           <at 0 0 RCIM_Cubic>
           {arriveTangentWeight=0.0247, interpMode=RCIM_Cubic, leaveTangentWeight=0.0255, tangentMode=RCTM_User, time=0.0174, value=0.05}
           {arriveTangentWeight=0.0589, interpMode=RCIM_Cubic, leaveTangentWeight=0.2882, tangentMode=RCTM_User, time=0.0565, value=0.05}
           {
            arriveTangent=3.9696
            arriveTangentWeight=0.0453
            interpMode=RCIM_Cubic
            leaveTangent=3.9696
            leaveTangentWeight=0.4261
            tangentMode=RCTM_User
            time=0.1319
            value=0.75
           }
           {arriveTangent=0.288, interpMode=RCIM_Cubic, leaveTangent=0.288, time=0.2254, value=0.8}
           {arriveTangentWeight=0.3013, interpMode=RCIM_Cubic, tangentMode=RCTM_User, time=1, value=1}
          ]}}})
      Z=&Multiply_Float(
       A=&FloatFromCurve(FloatCurve=DataInterface() {propertyValues {Curve {keys [
            {arriveTangent=779.4893, interpMode=RCIM_Cubic, leaveTangent=779.4893, leaveTangentWeight=3.3909, tangentMode=RCTM_User, time=0, value=0}
            {
             arriveTangent=19.032
             arriveTangentWeight=2.8007
             interpMode=RCIM_Cubic
             leaveTangent=19.032
             leaveTangentWeight=13.8693
             tangentMode=RCTM_User
             time=0.099
             value=30
            }
            {arriveTangentWeight=0.2773, interpMode=RCIM_Cubic, leaveTangentWeight=0.3745, tangentMode=RCTM_User, time=1.2045, value=45}
           ]}}})
       B=&RemapRange(
        Input Value To Remap=&VectorLength(Vector=&Subtract_Position(
          A=Linked() {linkedVariable {name=Particles.SkeletalMeshLocation001.SampledScale, type=Vector3f}}
          B=Linked() {linkedVariable {name=Particles.SkeletalMeshLocation002.SampledScale, type=Vector3f}}
         ))
        Input Min=0
        Input Max=100
        Output Min=0
        Output Max=2.5
       )
      )
     )
     B=&Divide_Vector(A:Vector3f=<1 1 1>, B:Vector3f=$Engine.Owner.Scale)
    ))
   &ScaleColor(
    Scale Alpha=&FloatFromCurve(
     FloatCurve=DataInterface() {propertyValues {Curve {keys [
         {interpMode=RCIM_Cubic, leaveTangent=0.0402, tangentMode=RCTM_Break, time=0, value=0}
         <at 0.3 0.0075>
         {arriveTangent=-0.3583, leaveTangent=-0.3583, time=2.9058, value=0.0075}
         <at 2.9074 0.0005>
         <at 2.9555 0.0005>
         <at 2.9764 0.0075>
         <at 3.1455 0.0075>
         <at 3.1461 0.0005>
         <at 3.2511 0.0005>
         <at 3.2699 0.0075>
         <at 3.3362 0.0075>
         <at 3.3431 0.0005>
         <at 3.4684 0.0005>
         <at 3.497 0.0075>
        ]}}}
     CurveIndex=$System.Age
    )
    ScaleA=true
    ScaleRGB=true
   )
   /SetVariables(Particles.TechnoBeamColor:LinearColor=<rgba 0.0282 0.7292 0.6943 0.01>) {isSetParameters=true}
   &FN_Light_Attributes(
    Color=$Particles.Color
    Create Color Attribute=true
    Create Exponent Attribute=true
    Create Position Attribute=true
    Create Radius Attribute=true
    Exponent=4
    Position=&AddVectorToPosition(Position=$Particles.Position, Vector=<0 0 -10>)
    Radius=90
    Scale Color=10
   )
   /SetVariables(Particles.DynamicMaterialParameterTechBeam:Vector4f=/Niagara/DynamicInputs/Multiply/Multiply_Vector4_ByFloat(
     Vector4:Vector4f=<0.8 0.8 1 1>
     Float:NiagaraFloat=&FloatFromCurve(
      FloatCurve=DataInterface() {propertyValues {Curve {keys [
          {interpMode=RCIM_Cubic, leaveTangent=1.6164, tangentMode=RCTM_Break, time=0, value=0}
          {arriveTangentWeight=0.173, interpMode=RCIM_Cubic, leaveTangentWeight=1.0013, tangentMode=RCTM_User, time=0.4503, value=1}
          {arriveTangent=-0.0121, leaveTangent=-0.3583, tangentMode=RCTM_User, time=3.0562, value=1}
          <at 3.0574 0.25>
          <at 3.1055 0.25>
          <at 3.1267 1>
          <at 3.2958 1>
          <at 3.2961 0.25>
          <at 3.4011 0.25>
          <at 3.4202 1>
          <at 3.4865 1>
          <at 3.4931 0.25>
          <at 3.6184 0.25>
          <at 3.6473 1>
         ]}}}
      CurveIndex=$Engine.System.Age
     )
    )) {isSetParameters=true}
   &DynamicMaterialParameters(
    Index 0 Param 4=&FloatFromCurve(
     FloatCurve=DataInterface() {propertyValues {Curve {keys [
         {interpMode=RCIM_Cubic, leaveTangent=1.6164, tangentMode=RCTM_Break, time=0, value=0}
         {arriveTangentWeight=0.173, interpMode=RCIM_Cubic, leaveTangentWeight=1.0013, tangentMode=RCTM_User, time=0.4503, value=1}
         {arriveTangent=-0.0121, leaveTangent=-0.3583, tangentMode=RCTM_User, time=3.0562, value=1}
         <at 3.0574 0.25>
         <at 3.1055 0.25>
         <at 3.1267 1>
         <at 3.2958 1>
         <at 3.2961 0.25>
         <at 3.4011 0.25>
         <at 3.4202 1>
         <at 3.4865 1>
         <at 3.4931 0.25>
         <at 3.6184 0.25>
         <at 3.6473 1>
        ]}}}
     CurveIndex=$Engine.System.Age
    )
    Index 0 Vector 1:Vector3f=<0.8 0.8 1>
    Module.Index0Float3Write=true
    Module.Index0VectorWrite=true
    Write Parameter Index 0=::NewEnumerator3
   )
  ]
  rendererClasses [&NiagaraLightRendererProperties, &NiagaraMeshRendererProperties]
  renderers [
   {rendererClass=&NiagaraLightRendererProperties, rendererIndex=0}
   {rendererClass=&NiagaraMeshRendererProperties, rendererIndex=1}
   {rendererClass=&NiagaraMeshRendererProperties, rendererIndex=2}
  ]
 }
 DeployConeLight {
  emitterUpdateScript [
   &EmitterState(Inactive Response=::NewEnumerator1, Life Cycle Mode=::NewEnumerator1, Loop Behavior=::NewEnumerator1, Loop Delay=1.6, UseLoopDelay=false)
   &SpawnBurst_Instantaneous(Use Spawn Probability=false, UseLoopCountLimit=false)
  ]
  particleSpawnScript [
   &InitializeParticle(
    Color=<rgba 0.1615 0.9403 1 0.7>
    Lifetime=3.54
    Mesh Scale=<0.8 0.8 0.8>
    Mesh Scale Mode=::NewEnumerator4
    Sprite Size Mode=::NewEnumerator1
    Uniform Sprite Size Max=15
    Uniform Sprite Size Min=10
    UsePositionOffset=false
    Write Material Random=true
   )
   /SetVariables(Particles.TinyRingsScale:Vector3f=&VectorFromFloat(Value=0.3)) {isSetParameters=true}
   /SetVariables(Particles.TinyRingsOrientation:Quat4f=&MakeQuaternion(Quaternion From=1, Axis=<1 0 -90>, Angle=0, Roll=180, From=?, To=?, Angle Type=::NewEnumerator0, Normalize Vectors=false)) {isSetParameters=true}
   /SetVariables(Particles.Linear Color=<rgba 0 1.9555 2 1>) {isSetParameters=true}
  ]
  particleUpdateScript [
   &SolveForcesAndVelocity(Clamp Velocity=false, Limit Acceleration=false)
   &ParticleState()
   &ScaleMeshSize(Scale Factor=&VectorFromCurve(VectorCurve=DataInterface() {propertyValues {
       XCurve {keys [
         {interpMode=RCIM_Cubic, leaveTangentWeight=0.0184, tangentMode=RCTM_User, time=0, value=0.1984}
         {arriveTangent=0.0461, interpMode=RCIM_Cubic, leaveTangent=0.0461, tangentMode=RCTM_User, time=0.0553, value=0.2}
         {arriveTangentWeight=0.0411, interpMode=RCIM_Cubic, leaveTangentWeight=0.3716, tangentMode=RCTM_User, time=0.1552, value=1}
         {arriveTangentWeight=0.2836, interpMode=RCIM_Cubic, tangentMode=RCTM_User, time=1, value=1}
        ]}
       YCurve {keys [
         {interpMode=RCIM_Cubic, leaveTangentWeight=0.0182, tangentMode=RCTM_User, time=0, value=0.2}
         {arriveTangent=0.0455, interpMode=RCIM_Cubic, leaveTangent=0.0455, tangentMode=RCTM_User, time=0.0545, value=0.2}
         {arriveTangentWeight=0.0414, interpMode=RCIM_Cubic, leaveTangentWeight=0.3714, tangentMode=RCTM_User, time=0.1552, value=1}
         {arriveTangentWeight=0.2836, interpMode=RCIM_Cubic, tangentMode=RCTM_User, time=1, value=1}
        ]}
       ZCurve {keys [
         {arriveTangent=4.252, interpMode=RCIM_Cubic, leaveTangent=4.252, tangentMode=RCTM_User, time=0, value=0}
         {arriveTangent=1.2037, interpMode=RCIM_Cubic, leaveTangent=1.2037, tangentMode=RCTM_User, time=0.1857, value=0.8}
         {arriveTangent=0.1842, interpMode=RCIM_Cubic, leaveTangent=0.1842, time=0.2993, value=0.9}
         {arriveTangentWeight=0.2759, interpMode=RCIM_Cubic, tangentMode=RCTM_User, time=1, value=0.95}
        ]}
      }}))
   &SkeletalMeshLocation(
    Bone Sampling Mode=::NewEnumerator4
    MaskSpawnGroup=false
    Orientation Sampling=::NewEnumerator1
    Sampled Position Offset=<0 0 -3>
    Skeletal Mesh=$System.Skeletal Mesh
   ) {module=SkeletalMeshLocation001}
   &ScaleColor(
    Scale Alpha=&FloatFromCurve(
     FloatCurve=DataInterface() {propertyValues {Curve {keys [
         {interpMode=RCIM_Cubic, leaveTangent=1.6164, tangentMode=RCTM_Break, time=0, value=0.1608}
         <at 0.3 0.7>
         {arriveTangent=-0.3583, leaveTangent=-0.3583, time=2.9058, value=0.7}
         <at 2.9221 0.2>
         <at 2.9702 0.2>
         <at 2.9764 0.7>
         <at 3.1455 0.7>
         <at 3.1609 0.2>
         <at 3.2658 0.2>
         <at 3.2699 0.7>
         <at 3.3362 0.7>
         <at 3.3578 0.2>
         <at 3.4831 0.2>
         <at 3.497 0.7>
        ]}}}
     CurveIndex=$Engine.System.Age
    )
    ScaleA=true
    ScaleRGB=true
   )
  ]
  rendererClasses [&NiagaraMeshRendererProperties]
  renderers [
   {rendererClass=&NiagaraMeshRendererProperties, rendererIndex=0}
   {rendererClass=&NiagaraMeshRendererProperties, rendererIndex=1}
   {rendererClass=&NiagaraMeshRendererProperties, rendererIndex=2}
  ]
 }
 GroundRings {
  emitterUpdateScript [
   &EmitterState(Inactive Response=::NewEnumerator1, Life Cycle Mode=::NewEnumerator1, Loop Behavior=::NewEnumerator1, Loop Duration=3.54, UseLoopDelay=false)
   /Game/Effects/Niagara/Modules/BlendSpace/BlendSpaceHelper()
   &SpawnRate(SpawnGroup=1, SpawnRate=2.6, Use Spawn Probability=false)
   &SpawnBurst_Instantaneous(Spawn Time=0.48, Use Spawn Probability=false, UseLoopCountLimit=false)
  ]
  particleSpawnScript [
   &InitializeParticle(
    Color=<rgba 0.2783 1 0.9522 1>
    Lifetime=3.4
    Mesh Scale=<0.65 0.65 0.35>
    Mesh Scale Mode=::NewEnumerator4
    Sprite Size Mode=::NewEnumerator3
    Uniform Sprite Size=250
    UsePositionOffset=true
    Write Material Random=true
   )
   &InitialMeshOrientation(Mesh Axis To Orient=<0 0 1>, Mesh Orientation Mode=::NewEnumerator2, Orientation Vector=$Engine.Owner.SystemZAxis, Use Rotation Vector=false)
   /SetVariables(Particles.VisibilityTag=&SpawnGroup()) {isSetParameters=true}
   /SetVariables(Particles.DecalFade=20) {isSetParameters=true}
   /SetVariables(
    Particles.FE Orientation:Quat4f=&OrientMeshToVector(Look At Direction=<0 0 1>, Reference Axis Vector=<1 0 0>)
    Particles.FE Position:NiagaraPosition=$Particles.Position
   ) {isSetParameters=true}
  ]
  particleUpdateScript [
   &ParticleState()
   /SetVariables(Particles.DecalSize:Vector3f=&VectorFromCurve(VectorCurve=DataInterface() {propertyValues {XCurve {keys [<at 0 125>,<at 0.2 235>]}, YCurve {keys [<at 0 125>,<at 0.2 235>]}, ZCurve {keys [<at 0 125>,<at 0.2 235>]}}})) {isSetParameters=true}
   &ScaleMeshSize(Scale Factor=&VectorFromCurve(VectorCurve=DataInterface() {propertyValues {
       XCurve {keys [<at 0 1.1 RCIM_Cubic>,<at 0.9947 1.2 RCIM_Cubic>]}
       YCurve {keys [<at 0 1.1 RCIM_Cubic>,<at 0.9947 1.2 RCIM_Cubic>]}
       ZCurve {keys [<at -0.0184 0 RCIM_Cubic>,<at 0.2868 1 RCIM_Cubic>]}
      }}))
   /SetVariables(Particles.DecalColor:LinearColor=&ColorFromCurve(ColorCurve=DataInterface() {propertyValues {
       AlphaCurve {keys [<at 0 0>,<at 0.2634 0.1>,<at 1 0>]}
       BlueCurve {keys [<at 0 0.8166>]}
       GreenCurve {keys [<at 0 1>]}
       RedCurve {keys [<at 0 0.2457>]}
      }})) {isSetParameters=true}
   &SkeletalMeshLocation(
    Bone Sampling Mode=::NewEnumerator4
    MaskSpawnGroup=false
    Orientation Sampling=::NewEnumerator1
    Position Sampling=::NewEnumerator2
    Sampled Position Offset Coordinate Space:ENiagaraSkelMeshTransforms=::NewEnumerator2
    Skeletal Mesh=$User.OwnerSkeleton
   )
   &self:Slope_Position_02(
    Apply Position=true
    Height Adjustment/Offset:NiagaraFloat=&MakeCustomFloatFromBool(
     Bool=&SetBoolByFloatComparison(A=&VectorLength(Vector=$Engine.Owner.Velocity), B=500)
     True Float=-20
     False Float=0
    )
    Slope Position:NiagaraPosition=$Particles.SkeletalMeshLocation.SampledPosition
    Trace Length=120
    Trace Origin Height=85
    bIsFrontEnd:NiagaraBool=$User.bIsFrontEndPreview
   ) {module=Slope_Position_02}
   /SetVariables(Particles.MeshOrientation=&OrientMeshToVector(Look At Direction=$Particles.CollisionNormal, Reference Axis Vector=<1 0 0>)) {isSetParameters=true}
   &ScaleColor(
    Linear Color Curve=DataInterface() {propertyValues {
      AlphaCurve {keys [<at 0 0>,<at 0.124 0.6>,<at 0.8718 0.6>,<at 1 0>]}
      BlueCurve {keys [<at 0 0.9522>]}
      GreenCurve {keys [<at 0 1>]}
      RedCurve {keys [<at 0 0.2783>]}
     }}
    Scale Mode=::NewEnumerator2
   )
   &BlendSpace(Rotation Contribution=0) {module=BlendSpace001}
   /SetVariables(
    Particles.MeshOrientation=/Niagara/DynamicInputs/TypeConversions/MakeCustomQuaternionFromBool(Bool:NiagaraBool=$User.bIsFrontEndPreview, True Quat:Quat4f=$Particles.FE Orientation, False Quat:Quat4f=$Particles.MeshOrientation)
    Particles.Position=/Niagara/DynamicInputs/TypeConversions/MakeCustomPositionFromBool(
     Bool:NiagaraBool=$User.bIsFrontEndPreview
     True Position:NiagaraPosition=$Particles.FE Position
     False Position:NiagaraPosition=$Particles.Position
    )
   ) {isSetParameters=true}
  ]
  rendererClasses [&NiagaraMeshRendererProperties, &NiagaraDecalRendererProperties]
  renderers [
   {rendererClass=&NiagaraMeshRendererProperties, rendererIndex=0}
   {rendererClass=&NiagaraDecalRendererProperties, rendererIndex=1}
  ]
 }
 LightningSmall {
  emitterUpdateScript [
   &EmitterState(Inactive Response=::NewEnumerator1, Life Cycle Mode=::NewEnumerator1, Loop Behavior=::NewEnumerator1, Loop Duration=3.35, UseLoopDelay=true)
   &SpawnBurst_Instantaneous(Spawn Count=2, Use Spawn Probability=false, UseLoopCountLimit=false) {enabled=false}
   &SpawnRate(SpawnRate=2, Use Spawn Probability=true)
  ]
  particleSpawnScript [
   &InitializeParticle(
    Color=<rgba 0.0083 0.7969 0.7487 1>
    Lifetime=&FloatFromCurve(FloatCurve=DataInterface() {propertyValues {Curve {keys [<at 0 2.3>,<at 1 2.5>]}}}, CurveIndex=$System.NormalizedLoopAge)
    Mesh Scale=<0.55 0.55 3>
    Mesh Scale Mode=::NewEnumerator4
    Sprite Size Mode=::NewEnumerator3
    UsePositionOffset=false
    Write Material Random=true
   )
   &CameraOffset(Camera Offset Amount=5)
   &ShapeLocation(
    Handle Radius=1
    Large Radius=&Multiply_Float(
     A=0.01
     B=&FN_LOD_Distance_Remap(Start Distance=350, End Distance=360, Minimum Output=0.7, Maximum Output=1, Clamp Output=true)
    )
    MaskSpawnGroup=false
    Non Uniform Scale=&Multiply_VectorByFloat(Vector=<1 1 1>, Float=1)
    Offset=&VectorFromCurve(
     VectorCurve=DataInterface() {propertyValues {XCurve {keys [<at 0 0>,<at 1 0>]}, YCurve {keys [<at 0 0>,<at 1 0>]}, ZCurve {keys [<at 0.1944 0>,<at 1 35>]}}}
     CurveIndex=$System.NormalizedLoopAge
    )
    Offset Mode=::NewEnumerator0
    Override Seed=false
    Rotation Axis=<0 0 -1>
    Rotation Mode=::NewEnumerator1
    Shape Primitive=::NewEnumerator3
    UseOwnerScale=true
   )
   &InitialMeshOrientation(
    Mesh Axis To Orient=<0 0 1>
    Mesh Orientation Mode=::NewEnumerator2
    Orientation Vector=$StackContext.Velocity
    Rotation=&RandomRangeVector(Minimum=<-1 0 0>, Maximum=<1 0 0>)
    Rotation Coordinate Space=::NewEnumerator2
    Use Rotation Vector=true
   ) {module=InitialMeshOrientation001}
  ]
  particleUpdateScript [
   &ParticleState()
   &ScaleMeshSize(Scale Factor=&VectorFromCurve(VectorCurve=DataInterface() {propertyValues {XCurve {keys [<at 0.0077 1.05>,<at 0.592 0.1>]}, YCurve {keys [<at 0.0077 1.05>,<at 0.5644 0.15>]}, ZCurve {keys [<at 0 1>,<at 1.0178 0>]}}})) {module=ScaleMeshSize001}
   &SkeletalMeshLocation(
    Bone Sampling Mode=::NewEnumerator4
    MaskSpawnGroup=false
    Orientation Sampling=::NewEnumerator1
    Position Sampling=::NewEnumerator2
    Skeletal Mesh=$System.Skeletal Mesh
   ) {module=SkeletalMeshLocation001}
   &PointAttractionForce(
    Attraction Radius=750
    Attraction Strength=&RandomRangeFloat(Minimum=18, Maximum=25)
    Attractor Position:NiagaraPosition=$Particles.SkeletalMeshLocation001.SampledPosition
    Attractor Position Offset:Vector3f=<0 0 -10>
    Kill Radius=20
    Kill Within Radius=true
    Module.Use Falloff=true
    UseAttractorOffset=true
   ) {module=PointAttractionForce001}
   &ScaleColor(
    Scale Alpha=&Multiply_Float(
     A=&RandomRangeFloat(Minimum=0.01, Maximum=0.2)
     B=&FloatFromCurve(FloatCurve=DataInterface() {propertyValues {Curve {keys [<at 0 0>,<at 0.1 0>,<at 0.2 0.4>,<at 0.6939 0.4>,<at 0.9979 0.2>]}}})
    )
    Scale RGB=&Multiply_VectorByFloat(
     Vector=<1 1 1>
     Float=&Multiply_Float(
      A=&RandomRangeFloat(Minimum=0, Maximum=1, Evaluation Type=::NewEnumerator1)
      B=&Multiply_Float(
       A=1
       B=&FloatFromCurve(
        FloatCurve=DataInterface() {propertyValues {Curve {keys [<at 0 10>]}}}
        Scale Curve=&FNTimeOfDayScalar(Night Scale=0.5, Morning Scale=0.5, Evening Scale=0.5)
       )
      )
     )
    )
    ScaleA=true
    ScaleRGB=true
   )
   &DynamicMaterialParameters(
    Index 0 Param 1=&FloatFromCurve(
     FloatCurve=DataInterface() {propertyValues {Curve {
        defaultValue=0
        keys [
         {arriveTangent=0.0118, interpMode=RCIM_Cubic, leaveTangent=0.0118, tangentMode=RCTM_User, time=0.0098, value=0}
         {
          arriveTangent=0.0001
          arriveTangentWeight=0.275
          interpMode=RCIM_Cubic
          leaveTangent=-0.0001
          leaveTangentWeight=0.2143
          tangentMode=RCTM_Break
          tangentWeightMode=RCTWM_WeightedBoth
          time=0.5098
          value=0.003
         }
         {arriveTangent=-0.0115, interpMode=RCIM_Cubic, leaveTangent=-0.0115, tangentMode=RCTM_User, time=1.0215, value=0}
        ]
       }}}
     Scale Curve=&RandomRangeFloat(Minimum=1, Maximum=1)
    )
    Index 0 Param 2=&FloatFromCurve(
     FloatCurve=DataInterface() {propertyValues {Curve {
        defaultValue=0
        keys [<at 0.0021 1 RCIM_Cubic>, {arriveTangent=0.0357, interpMode=RCIM_Cubic, leaveTangent=0.0357, tangentMode=RCTM_User, time=1, value=0.2}]
       }}}
     Scale Curve=&Multiply_Float(A=0.1, B=1)
    )
    Index 0 Param 3=&RandomRangeFloat(Minimum=0.25, Maximum=1)
    Index 0 Param 4=&FloatFromCurve(FloatCurve=DataInterface() {propertyValues {Curve {
        defaultValue=0
        keys [
         {arriveTangent=4.1785, interpMode=RCIM_Cubic, leaveTangent=4.1785, tangentMode=RCTM_User, time=0, value=0}
         {arriveTangent=0.0224, interpMode=RCIM_Cubic, leaveTangent=0.0224, tangentMode=RCTM_User, time=0.4948, value=0.475}
        ]
       }}})
    Index 1 Param 1=&FloatFromCurve(FloatCurve=DataInterface() {propertyValues {Curve {
        defaultValue=1
        keys [
         {interpMode=RCIM_Cubic, tangentMode=RCTM_User, time=0.0043, value=0}
         {arriveTangent=2.9607, interpMode=RCIM_Cubic, leaveTangent=2.9607, tangentMode=RCTM_User, time=0.7, value=1}
        ]
       }}})
    Index 1 Param 2=&FloatFromCurve(FloatCurve=DataInterface() {propertyValues {Curve {keys [<at 0 0>,<at 1 1>]}}})
    Index 1 Param 3=&Multiply_Float(A=0.5, B=0.7)
    Index 1 Param 4=&RandomRangeFloat(Minimum=0.5, Maximum=0.7)
    Module.Index0Float0Write=true
    Module.Index0Float1Write=true
    Module.Index0Float2Write=true
    Module.Index0Float3Write=true
    Module.Index1Float0Write=true
    Module.Index1Float1Write=true
    Module.Index1Float2Write=true
    Module.Index1Float3Write=true
    Write Parameter Index 1=::NewEnumerator1
   )
   &Drag(
    Drag=&FloatFromCurve(FloatCurve=DataInterface() {propertyValues {Curve {keys [<at 0 0 RCIM_Cubic>,<at 1 3 RCIM_Cubic>]}}})
    Use Linear Drag=true
    Use Rotational Drag=true
   )
   &SolveForcesAndVelocity(Clamp Velocity=false, Limit Acceleration=false)
  ]
  rendererClasses [&NiagaraMeshRendererProperties]
  renderers [{rendererClass=&NiagaraMeshRendererProperties, rendererIndex=0}]
 }
 Shards {
  emitterUpdateScript [
   &EmitterState(
    Inactive Response=::NewEnumerator1
    Life Cycle Mode=::NewEnumerator1
    Loop Behavior=::NewEnumerator1
    Loop Delay=0.3
    Loop Duration=3.5
    UseLoopDelay=true
   )
   &SpawnRate(
    SpawnRate=&FloatFromCurve(FloatCurve=DataInterface() {propertyValues {Curve {keys [<at 0 1>,<at 0.1125 7>]}}}, CurveIndex=$Emitter.NormalizedLoopAge)
    Use Spawn Probability=false
   ) {module=SpawnRate001}
  ]
  particleSpawnScript [
   &InitializeParticle(
    Color=<rgba 0.5 0.5 0.5 1>
    Lifetime Max=1
    Lifetime Min=0.8
    Lifetime Mode=::NewEnumerator1
    Mesh Scale Mode=::NewEnumerator1
    Mesh Uniform Scale Max=0.1
    Mesh Uniform Scale Min=0.05
    Position Offset=<0 0 -20>
    Sprite Size Mode=::NewEnumerator3
    Uniform Sprite Size=20
    UsePositionOffset=false
    Write Material Random=true
   )
   &TorusLocation(
    Align Torus With Axis=false
    Handle Radius=0
    Large Radius=80
    MaskSpawnGroup=false
    Module.Override Randomness=false
    Offset=&RandomRangeVector(
     Minimum=&RandomRangeVector(Minimum=<0 0 0>, Maximum=<0 0 40>)
     Maximum=&MakeVector(
      X=0
      Y=0
      Z=&FloatFromCurve(
       FloatCurve=DataInterface() {propertyValues {Curve {keys [
           {interpMode=RCIM_Cubic, tangentMode=RCTM_User, time=0, value=20}
           {arriveTangent=107.1069, interpMode=RCIM_Cubic, leaveTangent=107.1069, tangentMode=RCTM_User, time=0.2, value=60}
           <at 1 80>
          ]}}}
       CurveIndex=$Emitter.NormalizedLoopAge
      )
     )
    )
    Torus Origin:NiagaraPosition=&AddVectorToPosition(Position=$Particles.Position, Vector=<0 0 -20>)
   ) {module=TorusLocation001}
   &AddVelocity(
    Scale Added Velocity=&MakeVector(
     X=1
     Y=1
     Z=&FloatFromCurve(FloatCurve=DataInterface() {propertyValues {Curve {keys [<at 0 1>,<at 0.6 1>,<at 1 3.5>]}}}, CurveIndex=$Emitter.Age)
    )
    Velocity=&RandomRangeVector(Minimum=<0 0 25>, Maximum=<0 0 50>)
   )
   &InitialMeshOrientation(
    Rotation=&RandomRangeVector(Minimum=<0 0 0>, Maximum=<1 1 1>)
    Use Orientation Vector=true
    Use Rotation Vector=true
   )
  ]
  particleUpdateScript [
   &ParticleState()
   &Drag(Drag=&RandomRangeFloat(Minimum=4, Maximum=7), Use Linear Drag=true, Use Rotational Drag=false)
   &CurlNoiseForce(Noise Frequency=10, Noise Strength=&RandomRangeFloat(Minimum=50, Maximum=150), Pan Noise=false)
   &SolveForcesAndVelocity(Clamp Velocity=false, Limit Acceleration=false)
   &UpdateMeshOrientation(
    Rotation Rate=&FloatFromCurve(
     FloatCurve=DataInterface() {propertyValues {Curve {keys [
         {arriveTangent=-2.1159, interpMode=RCIM_Cubic, leaveTangent=-2.1159, tangentMode=RCTM_User, time=0, value=1}
         {interpMode=RCIM_Cubic, tangentMode=RCTM_Break, time=1, value=0.2}
        ]}}}
     Scale Curve=&RandomRangeFloat(Minimum=0.7, Maximum=1.5)
    )
    Rotation Vector=&RandomRangeVector(Minimum=<-1 -1 -1>, Maximum=<1 1 1>)
   )
   &ScaleMeshSize(Scale Factor=&Multiply_VectorByFloat(
     Vector=<1 1 1>
     Float=&FloatFromCurve(FloatCurve=DataInterface() {propertyValues {Curve {keys [
          {interpMode=RCIM_Cubic, tangentMode=RCTM_User, time=0, value=0}
          {interpMode=RCIM_Cubic, tangentMode=RCTM_User, time=0.3043, value=1}
          {tangentMode=RCTM_User, time=1, value=0}
         ]}}})
    ))
   &DynamicMaterialParameters(
    Index 0 Param 2=1
    Index 0 Param 3=1
    Index 0 Param 4=1
    Module.Index0Float0Write=true
    Module.Index0Float1Write=true
    Module.Index0Float2Write=true
    Module.Index0Float3Write=true
   )
  ]
  rendererClasses [&NiagaraMeshRendererProperties]
  renderers [{rendererClass=&NiagaraMeshRendererProperties, rendererIndex=0}]
 }
 Sparks {
  emitterUpdateScript [
   &EmitterState(Life Cycle Mode=::NewEnumerator1, Loop Behavior=::NewEnumerator1, Loop Delay=0.2, Loop Duration=3.2, UseLoopDelay=true)
   &SpawnRate(SpawnRate=8, Use Spawn Probability=true)
  ]
  particleSpawnScript [
   &InitializeParticle(
    Color=<rgba 0.2789 1 0.956 1>
    Lifetime Min=0.3
    Lifetime Mode=::NewEnumerator1
    Position Offset=&MakeVector(X=0, Y=0, Z=0)
    Sprite Size Mode=::NewEnumerator1
    Uniform Sprite Size Max=18
    Uniform Sprite Size Min=2
    UsePositionOffset=true
    Write Material Random=true
   )
   &ShapeLocation(
    Handle Radius=0
    Large Radius=65
    MaskSpawnGroup=false
    Non Uniform Scale=<1 1 -110>
    Override Seed=false
    Shape Primitive=::NewEnumerator3
    Surface Distribution=1
    UseOwnerScale=true
   )
   &AddVelocity(Velocity=&RandomRangeVector(Minimum=<0 0 25>, Maximum=<0 0 60>))
  ]
  particleUpdateScript [
   &ParticleState()
   &ScaleSpriteSize(
    Uniform Curve Scale=2
    Uniform Curve Sprite Scale=DataInterface() {propertyValues {Curve {keys [
        <at 0 0 RCIM_Cubic>
        <at 0.0116 1 RCIM_Cubic>
        {arriveTangent=-0.1423, interpMode=RCIM_Cubic, leaveTangent=-0.1423, tangentMode=RCTM_User, time=1, value=0}
       ]}}}
   )
   &ScaleColor(
    Linear Color Curve=DataInterface() {propertyValues {
      AlphaCurve {keys [<at 0 0>,<at 0.2402 0.4>,<at 1 0>]}
      BlueCurve {keys [<at 0 0.956>]}
      GreenCurve {keys [<at 0 1>]}
      RedCurve {keys [<at 0 0.2789>]}
     }}
    Scale Mode=::NewEnumerator2
   )
   &SkeletalMeshLocation(
    Bone Sampling Mode=::NewEnumerator4
    MaskSpawnGroup=false
    Orientation Sampling=::NewEnumerator1
    Position Sampling=::NewEnumerator2
    Skeletal Mesh=$System.Skeletal Mesh
   ) {module=SkeletalMeshLocation001}
   &SolveForcesAndVelocity(Clamp Velocity=false, Limit Acceleration=false)
  ]
  rendererClasses [&NiagaraSpriteRendererProperties]
  renderers [{rendererClass=&NiagaraSpriteRendererProperties}]
 }
}
summary {
 systemName=NS_Emote_Keytar_BeamMeUp_Beam
 emitters [
  {emitterName=ConeLights, rendererClasses [&NiagaraLightRendererProperties, &NiagaraMeshRendererProperties]}
  {emitterName=DeployConeLight, rendererClasses [&NiagaraMeshRendererProperties]}
  {emitterName=GroundRings, rendererClasses [&NiagaraMeshRendererProperties, &NiagaraDecalRendererProperties]}
  {emitterName=Sparks, rendererClasses [&NiagaraSpriteRendererProperties]}
  {emitterName=LightningSmall, rendererClasses [&NiagaraMeshRendererProperties]}
  {emitterName=Shards, rendererClasses [&NiagaraMeshRendererProperties]}
 ]
 userVariables [
  User.bIsFrontEndPreview {type=NiagaraBool}
  User.OwnerSkeleton {
   type=NiagaraDataInterfaceSkeletalMesh
   defaultValue {
    struct=NiagaraExt_VariableValue_DataInterface
    value {dataInterface=&self:NiagaraDataInterfaceSkeletalMesh_0}
   }
  }
 ]
}
