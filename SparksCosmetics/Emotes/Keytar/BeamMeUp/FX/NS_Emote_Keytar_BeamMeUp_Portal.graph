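# Niagara effect for the portal in the Keytar BeamMeUp emote in Sparks cosmetics, a short glowing cyan swirl with sparks.
# PortalSwirl bursts a combined sprite, mesh and light with curve-driven scale and fade, adjusting light brightness for front end versus gameplay.
# After short delays, LightningZappies spawns randomly facing SubUV arcs around a ring, while Motes adds small particles spun by vortex force and slowed by drag.
# The Flash emitter has no renderers.
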NiagaraSystem /SparksCosmetics/Emotes/Keytar/BeamMeUp/FX/NS_Emote_Keytar_BeamMeUp_Portal
dependencies {
 usedDataInterfaces [
  &NiagaraDataInterfaceCurve
  &NiagaraDataInterfaceColorCurve
  &NiagaraDataInterfaceSpriteRendererInfo
 ]
 usedDynamicInputs [&SimulationPosition, &UniformRangedVector2D]
 usedModules [
  &SystemState
  &self:SystemSpawnScript.SystemScriptSource.SystemScriptGraph.NiagaraNodeAssignment_1.SetVariables_#1
 ]
}
emitters {
 Flash {rendererClasses [], renderers []}
 LightningZappies {
  emitterUpdateScript [
   &EmitterState(
    Inactive Response=::NewEnumerator1
    Life Cycle Mode=::NewEnumerator1
    Loop Behavior=::NewEnumerator1
    Loop Delay=0.15
    Loop Duration=0.57
    UseLoopDelay=true
   )
   &SpawnRate(SpawnRate=25, Use Spawn Probability=false)
  ]
  particleSpawnScript [
   &InitializeParticle(
    Lifetime=0.5
    Position Offset=<0 0 5>
    Sprite Rotation Mode=::NewEnumerator2
    Sprite Size Mode=::NewEnumerator1
    Uniform Sprite Size Max=25
    Uniform Sprite Size Min=15
    UsePositionOffset=true
    Write Material Random=true
   ) {module=InitializeParticle001}
   &DynamicMaterialParameters(
    Index 0 Param 3=&FloatFromCurve(FloatCurve=DataInterface() {propertyValues {Curve {keys [<at 0 1 RCIM_Cubic>,<at 0.2 0.025 RCIM_Cubic>,<at 1 1>]}}})
    Module.Index0Float0Write=true
    Module.Index0Float1Write=true
    Module.Index0Float2Write=true
    Module.Index0Float3Write=true
   )
   &ShapeLocation(
    Handle Radius=1.4693
    Large Radius=13
    MaskSpawnGroup=false
    Override Seed=false
    Rotation Mode=::NewEnumerator2
    Shape Primitive=::NewEnumerator3
    UseOwnerScale=true
    Yaw / Pitch / Roll=<0 1.75 0>
   )
   &SpriteFacingAndAlignment(
    Sprite Facing=&RandomRangeVector(Minimum=<-1 -1 -1>, Maximum=<1 1 1>)
    WriteAlignment=false
    WriteFacing=true
   )
  ]
  particleUpdateScript [
   &ParticleState()
   &DynamicMaterialParameters(
    Index 0 Param 3=&FloatFromCurve(FloatCurve=DataInterface() {propertyValues {Curve {keys [<at 0 1 RCIM_Cubic>,<at 0.2 0.025 RCIM_Cubic>,<at 1 1>]}}})
    Module.Index0Float0Write=true
    Module.Index0Float1Write=true
    Module.Index0Float2Write=true
    Module.Index0Float3Write=true
   ) {module=DynamicMaterialParameters001}
   &ScaleColor(
    Scale Alpha=&FloatFromCurve(FloatCurve=DataInterface() {propertyValues {Curve {keys [<at 0.75 1 RCIM_Cubic>, {arriveTangent=-15.0784, interpMode=RCIM_Cubic, leaveTangent=-15.0784, tangentMode=RCTM_User, time=1, value=0}]}}})
    Scale RGB=&Multiply_VectorByFloat(
     Vector=&MakeVectorFromLinearColorRGB(Color=&ColorFromCurve(ColorCurve=DataInterface() {propertyValues {
         AlphaCurve {keys []}
         BlueCurve {keys [<at 0 1.2093>, {interpMode=RCIM_Cubic, leaveTangent=-0.9012, time=0.3527, value=1.1938}]}
         GreenCurve {keys [<at 0 1.5>, {interpMode=RCIM_Cubic, leaveTangent=-0.7396, time=0.3527, value=1.5}]}
         RedCurve {keys [<at 0 0.4655>, {interpMode=RCIM_Cubic, leaveTangent=-0.943, time=0.3527, value=0.5181}]}
        }}))
     Float=1
    )
    ScaleA=true
    ScaleRGB=true
   )
   &SubUVAnimation(
    Evaluation Type=::NewEnumerator0
    "Sprite Renderer "=DataInterface() {propertyValues {SpriteRenderer=&self:LightningZappies_0.NiagaraSpriteRendererProperties_1}}
    SubUV Animation Mode:ENiagara_SubUVLookupModeV2=::NewEnumerator2
    UseEndFrame=false
    UseStartFrame=false
   ) {module=SubUVAnimation001}
  ]
  rendererClasses [&NiagaraSpriteRendererProperties]
  renderers [{rendererClass=&NiagaraSpriteRendererProperties}]
 }
 Motes {
  emitterUpdateScript [
   &EmitterState(Life Cycle Mode=::NewEnumerator1, Loop Behavior=::NewEnumerator1, Loop Delay=0.15, Loop Duration=0.65, UseLoopDelay=true)
   &SpawnRate(SpawnRate=60, Use Spawn Probability=false)
  ]
  particleSpawnScript [
   &InitializeParticle(
    Lifetime Max=0.35
    Lifetime Min=0.2
    Lifetime Mode=::NewEnumerator1
    Sprite Size Mode=::NewEnumerator1
    Uniform Sprite Size Max=3
    Uniform Sprite Size Min=1
    UsePositionOffset=false
    Write Material Random=true
   )
   &ShapeLocation(
    Handle Radius=1
    Large Radius=15
    MaskSpawnGroup=false
    Offset=<2.6752 0 8>
    Offset Mode=::NewEnumerator0
    Override Seed=false
    Rotation Mode=::NewEnumerator2
    Shape Primitive=::NewEnumerator3
    UseOwnerScale=true
    Yaw / Pitch / Roll=<0 0.25 0>
   )
   &VortexForce(UseOriginOffset=false, Vortex Axis=<-1 0 0>, Vortex Force Amount=250)
   &ApplyInitialForces()
  ]
  particleUpdateScript [
   &ParticleState()
   &ScaleSpriteSize(Uniform Curve Sprite Scale=DataInterface() {propertyValues {Curve {keys [
        <at 0 0 RCIM_Cubic>
        {arriveTangentWeight=0.1158, interpMode=RCIM_Cubic, leaveTangentWeight=0.2906, tangentMode=RCTM_User, time=0.1988, value=1}
        {arriveTangentWeight=0.2658, interpMode=RCIM_Cubic, leaveTangentWeight=0.1613, tangentMode=RCTM_User, time=0.6974, value=1}
        <at 1 0 RCIM_Cubic>
       ]}}})
   &Drag(
    Drag=&FloatFromCurve(FloatCurve=DataInterface() {propertyValues {Curve {keys [<at 0 0>,<at 0.2459 15>]}}})
    Use Linear Drag=true
    Use Rotational Drag=true
   )
   &SolveForcesAndVelocity(Clamp Velocity=false, Limit Acceleration=false)
   &ScaleColor(
    Linear Color Curve=DataInterface() {propertyValues {
      AlphaCurve {keys [<at 0 0>,<at 0.1012 1>,<at 1 0>]}
      BlueCurve {keys [<at 0 2>]}
      GreenCurve {keys [<at 0 1.8577>]}
      RedCurve {keys [<at 0 0>]}
     }}
    Scale Mode=::NewEnumerator2
   )
  ]
  rendererClasses [&NiagaraSpriteRendererProperties]
  renderers [{rendererClass=&NiagaraSpriteRendererProperties}]
 }
 PortalSwirl {
  emitterUpdateScript [
   &EmitterState(Life Cycle Mode=::NewEnumerator1, Loop Behavior=::NewEnumerator1, Loop Delay=0.1, Loop Duration=2.85, UseLoopDelay=true)
   &SpawnBurst_Instantaneous(Use Spawn Probability=false, UseLoopCountLimit=false)
  ]
  particleSpawnScript [
   &InitializeParticle(
    Color=<rgba 0.5565 2 1.9044 0.05>
    Lifetime=0.7
    Mesh Scale Mode=::NewEnumerator3
    Mesh Uniform Scale=0.7
    Position Offset=<0 0 5>
    Sprite Size Mode=::NewEnumerator3
    Uniform Sprite Size=100
    UsePositionOffset=true
    Write Material Random=true
   )
   &CameraOffset(Camera Offset Amount=20)
   &SpriteFacingAndAlignment(WriteAlignment=false, WriteFacing=true)
   &InitialMeshOrientation(Mesh Orientation Mode=::NewEnumerator6, Orientation Quaternion:Quat4f=<0 -1 0 1>, Use Rotation Vector=false)
  ]
  particleUpdateScript [
   &ParticleState()
   &ScaleSpriteSize(Uniform Curve Sprite Scale=DataInterface() {propertyValues {Curve {keys [
        {interpMode=RCIM_Cubic, leaveTangentWeight=0.5835, tangentMode=RCTM_User, time=0.0021, value=0.4}
        {arriveTangent=1.07, interpMode=RCIM_Cubic, leaveTangent=1.07, time=0.1121, value=0.65}
        {arriveTangentWeight=0.1037, interpMode=RCIM_Cubic, leaveTangentWeight=0.1331, tangentMode=RCTM_User, time=0.3292, value=0.75}
        {arriveTangentWeight=0.1366, interpMode=RCIM_Cubic, leaveTangentWeight=0.1054, tangentMode=RCTM_User, time=0.7971, value=0.75}
        {arriveTangent=-9.3479, interpMode=RCIM_Cubic, leaveTangent=-9.3479, tangentMode=RCTM_User, time=1, value=0}
       ]}}})
   &FN_Light_Attributes(
    Color=<rgba 0 0.9288 1 1>
    Create Color Attribute=true
    Create Exponent Attribute=true
    Create Position Attribute=true
    Create Radius Attribute=true
    Enable Particle Light Enable=true
    Exponent=2
    Light Enabled=true
    Radius=130
    Scale Color=&FloatFromCurve(
     FloatCurve=DataInterface() {propertyValues {Curve {keys [
         <at 0 0 RCIM_Cubic>
         {arriveTangentWeight=0.0581, interpMode=RCIM_Cubic, leaveTangentWeight=0.349, tangentMode=RCTM_User, time=0.0705, value=1}
         {arriveTangent=-0.7253, interpMode=RCIM_Cubic, leaveTangent=-0.7253, tangentMode=RCTM_User, time=0.9036, value=0.8288}
         {arriveTangent=-13.426, interpMode=RCIM_Cubic, leaveTangent=-13.426, tangentMode=RCTM_User, time=1, value=0}
        ]}}}
     Scale Curve=&FN_FrontEndFloatModifier(Gameplay=2, Front End=1)
    )
   )
   &ScaleColor(
    Scale Alpha=&FloatFromCurve(FloatCurve=DataInterface() {propertyValues {Curve {keys [
         {arriveTangent=1.9842, interpMode=RCIM_Cubic, leaveTangent=1.9842, tangentMode=RCTM_User, time=0.2, value=0}
         {
          arriveTangentWeight=0.0436
          interpMode=RCIM_Cubic
          leaveTangentWeight=0.1587
          tangentMode=RCTM_User
          tangentWeightMode=RCTWM_WeightedBoth
          time=0.3307
          value=0.6
         }
         {arriveTangentWeight=0.2131, interpMode=RCIM_Cubic, leaveTangentWeight=0.0865, tangentMode=RCTM_User, time=0.8068, value=0.6}
         {tangentMode=RCTM_User, time=1, value=0}
        ]}}})
    ScaleA=true
    ScaleRGB=true
   )
   /SetVariables(Particles.VortexScale:Vector3f=&Multiply_VectorByFloat(
     Vector=<1 1 1>
     Float=&FloatFromCurve(FloatCurve=DataInterface() {propertyValues {Curve {keys [
          {interpMode=RCIM_Cubic, leaveTangentWeight=0.1097, tangentMode=RCTM_User, time=0, value=0.15}
          {
           arriveTangent=0.8466
           arriveTangentWeight=0.1037
           interpMode=RCIM_Cubic
           leaveTangent=0.8466
           leaveTangentWeight=0.1331
           tangentMode=RCTM_User
           time=0.2015
           value=0.75
          }
          {
           arriveTangent=-0.7638
           arriveTangentWeight=0.1366
           interpMode=RCIM_Cubic
           leaveTangent=-0.7638
           leaveTangentWeight=0.1054
           tangentMode=RCTM_User
           time=0.7971
           value=0.75
          }
          {arriveTangentWeight=0.6359, interpMode=RCIM_Cubic, tangentMode=RCTM_User, time=1, value=0}
         ]}}})
    )) {isSetParameters=true}
   /SetVariables() {isSetParameters=true}
   /SetVariables(Particles.DynamicMaterialParameter001:Vector4f=&MakeVector4(X=1, Y=0, Z=1, W=1)) {isSetParameters=true}
   &DynamicMaterialParameters(
    Index 0 Param 1=1
    Index 0 Param 2=1
    Module.Index0Float0Write=true
    Module.Index0Float1Write=true
    Module.Index0Float2Write=true
    Module.Index0Float3Write=true
   )
  ]
  rendererClasses [
   &NiagaraSpriteRendererProperties
   &NiagaraLightRendererProperties
   &NiagaraMeshRendererProperties
  ]
  renderers [
   {rendererClass=&NiagaraSpriteRendererProperties}
   {rendererClass=&NiagaraLightRendererProperties, rendererIndex=1}
   {rendererClass=&NiagaraMeshRendererProperties, rendererIndex=2}
  ]
 }
}
summary {
 systemName=NS_Emote_Keytar_BeamMeUp_Portal
 emitters [
  {
   emitterName=PortalSwirl
   rendererClasses [
    &NiagaraSpriteRendererProperties
    &NiagaraLightRendererProperties
    &NiagaraMeshRendererProperties
   ]
  }
  {emitterName=LightningZappies, rendererClasses [&NiagaraSpriteRendererProperties]}
  {emitterName=Flash, rendererClasses []}
  {emitterName=Motes, rendererClasses [&NiagaraSpriteRendererProperties]}
 ]
 userVariables [
  User.bIsFrontEnd {type=NiagaraBool}
  User.bIsFrontEndPreview {type=NiagaraBool}
  User.RibbonIDOffset {type=NiagaraInt32}
  User.SkeletalMesh {
   type=NiagaraDataInterfaceSkeletalMesh
   defaultValue {
    struct=NiagaraExt_VariableValue_DataInterface
    value {dataInterface=&self:NiagaraDataInterfaceSkeletalMesh_0}
   }
  }
 ]
}
