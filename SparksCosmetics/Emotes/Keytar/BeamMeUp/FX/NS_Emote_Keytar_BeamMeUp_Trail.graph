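# Niagara trail effect for the BeamMeUp keytar emote, drawing short-lived, fading and narrowing ribbons behind what the names indicate is a UFO.
# UFO_Trail's spawn rate follows a System.Age on-off-on curve, lerped toward zero as owner speed rises.
# Color combines a cyan gradient with time-of-day tint, and BlendSpace switches world/local position blending on an owner-speed threshold.
# UFO_Glow and TractorBeam_Glow emitters have no renderers.
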NiagaraSystem /SparksCosmetics/Emotes/Keytar/BeamMeUp/FX/NS_Emote_Keytar_BeamMeUp_Trail
dependencies {
 usedDataInterfaces [&NiagaraDataInterfaceCurve, &NiagaraDataInterfaceColorCurve]
 usedDynamicInputs []
 usedModules [&SystemState, &BlendSpaceHelper]
}
emitters {
 TractorBeam_Glow {rendererClasses [], renderers []}
 UFO_Glow {rendererClasses [], renderers []}
 UFO_Trail {
  emitterUpdateScript [
   &EmitterState(Life Cycle Mode=::NewEnumerator1, Loop Behavior=::NewEnumerator1, Loop Delay=$System.Delay, Loop Duration=10, UseLoopDelay=false)
   /SetVariables(Emitter.NewSpawnRate:NiagaraFloat=&Lerp_Float(
     A=&FloatFromCurve(
      FloatCurve=DataInterface() {propertyValues {Curve {keys [<at 0 40>,<at 0.9603 40>,<at 0.9604 0>,<at 5.047 0>,<at 5.0471 40>,<at 5.858 40>]}}}
      CurveIndex=$System.Age
     )
     B=0
     Alpha=&RemapRange(
      Input Value To Remap=&VectorLength(Vector=$Engine.Owner.Velocity)
      Input Min=0
      Input Max=500
      Output Min=0
      Output Max=1
     )
    )) {isSetParameters=true}
   &SpawnRate(SpawnRate=$Emitter.NewSpawnRate, Use Spawn Probability=false)
  ]
  particleSpawnScript [
   &InitializeParticle(
    Lifetime=0.3
    Position Offset=<0 -5 3>
    Ribbon Facing Mode=::NewEnumerator1
    Ribbon Facing Vector:Vector3f=$Engine.Owner.SystemZAxis
    Ribbon Width=10
    Ribbon Width Mode=::NewEnumerator1
    UsePositionOffset=true
    Write Material Random=true
   )
   &AddVelocity(Velocity=<0 -15 0>)
  ]
  particleUpdateScript [
   &self:Leading_Edge_Fix() {module=Leading_Edge_Fix}
   &ParticleState()
   &ScaleColor(
    Scale Alpha=&FloatFromCurve(FloatCurve=DataInterface() {propertyValues {Curve {keys [<at 0 1>,<at 1 0>]}}})
    Scale RGB=&Multiply_Vector(
     A=&MakeVectorFromLinearColorRGB(Color=&ColorFromCurve(ColorCurve=DataInterface() {propertyValues {
         AlphaCurve {keys []}
         BlueCurve {keys [<at 0 1 RCIM_Cubic>,<at 0.6974 1>]}
         GreenCurve {keys [<at 0 0.9288 RCIM_Cubic>,<at 0.6974 0.9403>]}
         RedCurve {keys [<at 0 0 RCIM_Cubic>,<at 0.6974 0.1615>]}
        }}))
     B=&MakeVectorFromLinearColorRGB(Color=&FNTimeOfDayColor(
       Night Color=<rgba 0.3169 0.4621 1 1>
       Morning Color=<rgba 1 0.8374 0.7324 1>
       Day Color=<rgba 1 1 1 1>
       Evening Color=<rgba 0.736 0.9073 1 1>
      ))
    )
    ScaleA=true
    ScaleRGB=true
   )
   &SolveForcesAndVelocity(Clamp Velocity=false, Limit Acceleration=false)
   &ScaleRibbonWidth(Ribbon Width Scale=&FloatFromCurve(FloatCurve=DataInterface() {propertyValues {Curve {keys [<at 0 1>,<at 1 0>]}}}))
   &BlendSpace("Position Blend (World <-> Local)":NiagaraFloat=&MakeCustomFloatFromBool(
     Bool=&SetBoolByFloatComparison(A=&VectorLength(Vector=$Engine.Owner.Velocity), B=310)
     True Float=1
     False Float=0
     Translation Contribution:NiagaraFloat=1
     Rotation Contribution:NiagaraFloat=1
    ))
  ]
  rendererClasses [&NiagaraRibbonRendererProperties]
  renderers [
   {rendererClass=&NiagaraRibbonRendererProperties}
   {rendererClass=&NiagaraRibbonRendererProperties, rendererIndex=1}
  ]
 }
}
summary {
 systemName=NS_Emote_Keytar_BeamMeUp_Trail
 emitters [
  {emitterName=UFO_Trail, rendererClasses [&NiagaraRibbonRendererProperties]}
  {emitterName=UFO_Glow, rendererClasses []}
  {emitterName=TractorBeam_Glow, rendererClasses []}
 ]
 userVariables [User.RibbonIDOffset {type=NiagaraInt32}]
}
